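# Niagara system for the GnatGala emote, named as the effect for when the emote's prop despawns.
# It holds a single emitter, SmokeLWE001, whose name suggests a lightweight smoke emitter.
# The system uses the SystemState module and exposes no user variables.
# No renderers, data interfaces or dynamic inputs are recorded for it.
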NiagaraSystem /BRCosmetics/Animation/Game/MainPlayer/Emotes/GnatGala/FX/NS_Emote_GnatGala_PropDespawn
dependencies {usedDataInterfaces [], usedDynamicInputs [], usedModules [&SystemState]}
emitters {SmokeLWE001 {rendererClasses [], renderers []}}
summary {systemName=NS_Emote_GnatGala_PropDespawn, emitters [{emitterName=SmokeLWE001, rendererClasses []}], userVariables []}
